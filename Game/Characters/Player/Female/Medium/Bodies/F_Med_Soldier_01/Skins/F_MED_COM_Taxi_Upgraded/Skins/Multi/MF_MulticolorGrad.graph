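# Material function for the upgraded Taxi character skin that changes multicolor gradient colors as the horizontal camera direction rotates.
# Camera direction becomes a normalized yaw angle, quantized by Phases into stepped V offsets for the gradient texture.
# Two samples with staggered steps, the first offset by Phase Shift, blend through HeightLerp.
# A triangle wave of the angle drives the transition, with a scalar height input.
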MaterialFunction /Game/Characters/Player/Female/Medium/Bodies/F_Med_Soldier_01/Skins/F_MED_COM_Taxi_Upgraded/Skins/Multi/MF_MulticolorGrad
graph {nodes [
  n1 = MaterialFunctionCall(Float2=FunctionInput(Preview=TextureCoordinate()) {inputName=UV, inputType=FunctionInput_Vector2}) {materialFunction=&BreakOutFloat2Components}
  n2 = FunctionInput(Preview=Constant() {r=2}) {inputName=Phases, inputType=FunctionInput_Scalar, previewValue=<1 0 0 1>}
  n3 = Divide(B=@n2)
  n4 = MaterialFunctionCall() {materialFunction=&Pi}
  n5 = MaterialFunctionCall() {materialFunction=&CameraDirectionVector}
  n6 = Normalize(VectorInput=AppendVector(A=ComponentMask(None=@n5.Result) {g=true, r=true}, B=Constant()))
  n7 = Constant3Vector() {constant=<rgba 1 0 0 0>}
  n8 = Arctangent2Fast(Y=DotProduct(A=Constant3Vector() {constant=<rgba 0 0 1 0>}, B=CrossProduct(A=@n6, B=@n7)), X=DotProduct(A=@n7, B=@n6))
  n9 = Subtract(A=@n4.Result, B=Multiply(A=@n8) {constB=-1})
  n10 = Divide(
   A=LinearInterpolate(A=Add(A=@n4.Result, B=@n9), B=@n8, Alpha=Step(X=@n8) {constX=0})
   B=MaterialFunctionCall(Multiplier=Constant() {r=2}) {materialFunction=&Pi}.Result
  )
  n11 = Multiply(A=@n2, B=@n10)
  n12 = Constant() {r=2}
  n13 = Multiply(A=@n2, B=@n12)
  n14 = Constant() {r=1}
  n15 = Add(
   A=Multiply(A=@n3, B=Floor(None=@n11))
   B=Multiply(A=Divide(B=@n13), B=FunctionInput(Preview=@n14) {inputName=Phase Shift, inputType=FunctionInput_Scalar, previewValue=<1 0 0 1>})
  )
  n16 = TextureObject() {texture=&/T_multicolor_gradient_v2_D}
  n17 = FunctionInput(Preview=@n16) {inputName=Gradient, inputType=FunctionInput_Texture2D, previewValue=<1 0 0 1>}
  n18 = TextureSample(
   UVs=MaterialFunctionCall(X=@n1.R, Y=Add(A=@n15, B=@n1.G)) {materialFunction=&MakeFloat2}.Result
   Tex=@n17
  ) {texture=&T_multicolor_gradient_v2_D}
  n19 = Add(A=@n10, B=Divide(B=@n13)) {constB=0.166}
  n20 = Multiply(A=@n3, B=Floor(None=Multiply(A=@n2, B=@n19)))
  n21 = TextureSample(
   UVs=MaterialFunctionCall(X=@n1.R, Y=Add(A=@n20, B=@n1.G)) {materialFunction=&MakeFloat2}.Result
   Tex=@n17
  ) {texture=&T_multicolor_gradient_v2_D}
  n22 = Subtract(A=Fmod(A=Multiply(A=@n13, B=@n10), B=@n12))
  n23 = Constant() {r=0.5}
  FunctionOutput(None=MaterialFunctionCall(
    A=@n18.RGB
    B=@n21.RGB
    Transition Phase=Abs(None=@n22)
    Height Texture=FunctionInput(Preview=@n23) {inputName=height texture, inputType=FunctionInput_Scalar, previewValue=<0.5 0 0 1>}
   ) {materialFunction=&HeightLerp}.Results)
 ]}
