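# Material function for a player character skin that blends a multicolor gradient into a material's base color and emissive.
# An alpha-masked Fresnel term plus bounds-normalized pre-skinned position on a selectable axis forms the UV for MF_MulticolorGrad, alongside a distortion input.
# The same Fresnel mask drives the blend, optionally dampened on downward-facing pre-skinned normals.
# Using pre-skinned position and normals keeps the gradient's positional component and dampening independent of animation.
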MaterialFunction /Game/Characters/Player/Female/Medium/Bodies/F_Med_Soldier_01/Skins/F_MED_COM_Taxi_Upgraded/Skins/Multi/MF_Multicolor_01
graph {nodes [
  n1 = GetMaterialAttributes(None=FunctionInput() {inputName=Material, inputType=FunctionInput_MaterialAttributes}) {attributeGetTypes [#1, #2]}
  n2 = Fresnel(ExponentIn=ScalarParameter() {defaultValue=1, group=Multicolor, parameterName=MC_Fresnel, sortPriority=10}) {baseReflectFraction=0, exponent=1}
  n3 = ScalarParameter() {defaultValue=1, group=Multicolor, parameterName=MC_FresnelStrength, sortPriority=11}
  n4 = Multiply(A=FunctionInput() {inputName=Alpha, inputType=FunctionInput_Scalar}, B=Saturate(None=Multiply(A=@n2, B=@n3) {constB=2})) {constB=2}
  n5 = ScalarParameter() {defaultValue=1, group=Multicolor, parameterName=MC_UV_ViewAngle, sortPriority=22}
  n6 = PreSkinnedPosition()
  n7 = ChannelMaskParameter(None=Divide(A=VertexInterpolator(VS=@n6).PS, B=PreSkinnedLocalBounds().Max)) {group=Multicolor, maskChannel=Blue, parameterName=MC_UV_Axis, sortPriority=19, defaultValue=<rgba 0 0 1 0>}
  n8 = Multiply(A=@n7.R, B=ScalarParameter() {defaultValue=1, group=Multicolor, parameterName=MC_UV_Tiling, sortPriority=20})
  n9 = ScalarParameter() {group=Multicolor, parameterName=MC_UV_Offset, sortPriority=21}
  n10 = TextureObjectParameter() {
   group=Multicolor
   parameterName=MC_GradientTex
   sortPriority=0
   texture=&/T_multicolor_gradient_v2_D
  }
  n11 = MaterialFunctionCall!(
   Phase Shift=ScalarParameter() {defaultValue=1, group=Multicolor, parameterName=MC_GradientShift, sortPriority=41}
   UV=Add(A=Multiply(A=@n4, B=@n5), B=Add(A=@n8, B=@n9))
   height texture=FunctionInput(Preview=Constant() {r=0.5}) {inputName=Distort, inputType=FunctionInput_Scalar}
   Gradient=@n10
   Phases=ScalarParameter() {defaultValue=4, group=Multicolor, parameterName=MC_GradientPhases, sortPriority=40}
  ) {materialFunction=&/MF_MulticolorGrad}
  n12 = PreSkinnedNormal()
  n13 = Multiply(
   A=DotProduct(A=VertexInterpolator(VS=@n12).PS, B=Constant3Vector() {constant=<rgba 0 0 -1 0>})
   B=ScalarParameter() {group=Multicolor, parameterName=MC_DownNormalDampen, sortPriority=100}
  )
  n14 = Multiply(A=OneMinus(None=Saturate(None=@n13)), B=@n4)
  n15 = LinearInterpolate(
   A=@n1.BaseColor
   B=Add(A=@n11.Result, B=ScalarParameter() {group=Multicolor, parameterName=MC_BaseColorBrighten, sortPriority=31})
   Alpha=@n14
  )
  n16 = Multiply(A=ScalarParameter() {defaultValue=1, group=Multicolor, parameterName=MC_EmissiveStrength}, B=@n14)
  FunctionOutput(None=SetMaterialAttributes!(
    MaterialAttributes=@n1.MaterialAttributes
    BaseColor=Saturate(None=@n15)
    EmissiveColor=LinearInterpolate(A=@n1.EmissiveColor, B=@n11.Result, Alpha=@n16)
   ) {attributeSetTypes [#1, #2]})
  SmoothStep()
 ]}
