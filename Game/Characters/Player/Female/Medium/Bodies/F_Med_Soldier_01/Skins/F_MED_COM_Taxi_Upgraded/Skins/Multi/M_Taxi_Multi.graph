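# Character skin material for the female Taxi Upgraded "Multi" skin, supporting skeletal meshes with Default Lit and Subsurface Profile shading.
# It chains material attributes from the shared character PreFX function through MF_Silk and MF_Multicolor_01 effects, then into the PostFX and cinematic PostFX functions.
# A single SkinFX_Mask texture drives the effects, with parameterized channel selection for silk, multicolor and multicolor distortion.
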Material /Game/Characters/Player/Female/Medium/Bodies/F_Med_Soldier_01/Skins/F_MED_COM_Taxi_Upgraded/Skins/Multi/M_Taxi_Multi
graph {
 properties {
  bUseMaterialAttributes=true
  bUsedWithMeshDeformer=true
  bUsedWithMorphTargets=true
  bUsedWithSkeletalMesh=true
  pixelDepthOffsetMode=PDOM_Legacy
  shadingModel=MSM_FromMaterialExpression
  usedShadingModels=Default Lit | Subsurface Profile
 }
 nodes [
  n1 = MaterialFunctionCall() {materialFunction=&BaseFN_CharacterFunction_PreFX}
  n2 = TextureSample(UVs=TextureCoordinate(), Tex=TextureObjectParameter() {group=Skin FX, parameterName=SkinFX_Mask})
  n3 = MaterialFunctionCall!(
   Material=MaterialFunctionCall!(Alpha=ChannelMaskParameter(None=@n2.RGBA) {group=Skin FX, parameterName=Silk Channel}.R, Material=@n1.Result) {materialFunction=/Game/Weapons/WeaponSkins/MaterialFunction/MF_Silk}.Result
   Distort=ChannelMaskParameter(None=@n2.RGBA) {group=Multicolor, parameterName=MC_DistortionChannel, sortPriority=90}.R
   Alpha=ChannelMaskParameter(None=@n2.RGBA) {group=Skin FX, maskChannel=Green, parameterName=Multicolor Channel, defaultValue=<rgba 0 1 0 0>}.R
  ) {materialFunction=&/MF_Multicolor_01}
  n4 = MaterialFunctionCall(AO=@n1.AO, Distance Blend=@n1.AO, Fresnel=@n1.AO, Gmap=@n1.AO, Pre Skinned Local Position=@n1.AO, Material=@n3.Result) {materialFunction=&BaseFN_CharacterFunction_PostFX}
  MaterialFunctionCall(In=@n4.Result) {materialFunction=&CINE_MF_CharPostFX}
 ]
}
